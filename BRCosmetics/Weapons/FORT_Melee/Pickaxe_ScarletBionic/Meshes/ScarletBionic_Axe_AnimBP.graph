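# Animation Blueprint for the Scarlet Bionic pickaxe that adds floating and springy secondary motion to its blade bone.
# Two chained spring controllers split translation axes, then an additive bone-space transform applies a time-based sine bob in Y and Z.
# The event graph inverts the melee-animation flag, disabling both effects during swings.
# Spring alpha is scaled very low and drops to zero above a pawn speed threshold.
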AnimBlueprint /BRCosmetics/Weapons/FORT_Melee/Pickaxe_ScarletBionic/Meshes/ScarletBionic_Axe_AnimBP
target_skeleton=&/ScarletBionic_Axe_Skeleton
graphs [
 AnimGraph {nodes [
   AnimGraphNode_SubInput_0 = Input Pose()
   n1 = Spring controller - Bone: dyn_Blade_jnt!(
    ComponentPose:Component Space Pose Link Structure=Spring controller - Bone: dyn_Blade_jnt(ComponentPose=Local To Component(LocalPose=@AnimGraphNode_SubInput_0), Alpha=Get SpringAlpha() -> (SpringAlpha:double)) {bTranslateZ=false, springDamping=2.5, alphaBoolBlend {blendOption=Linear}, springBone {boneName=dyn_Blade_jnt}}
    Alpha:float=Get SpringAlpha()
   ) -> (Pose:Component Space Pose Link Structure) {
    is=title AnimGraphNode_SpringBone
    bTranslateX=false
    bTranslateY=false
    springDamping=2.5
    springStiffness=120
    alphaBoolBlend {blendOption=Linear}
    springBone {boneName=dyn_Blade_jnt}
   }
   n2 = "Transform (Modify) Bone - Bone: dyn_Blade_jnt"(
    ComponentPose=@n1
    Alpha=Get Is Melee Anim Float() -> (Is Melee Anim Float:float)
    Translation=Get Blade Floating() -> (Blade Floating:Vector)
    Rotation="0,0,0"
    Scale="1,1,1"
   ) {
    translationMode=BMM_Additive
    translationSpace=BCS_BoneSpace
    alphaBoolBlend {blendOption=Linear}
    alphaScaleBiasClamp {bInterpResult=true}
    boneToModify {boneName=dyn_Blade_jnt}
   }
   Output Pose(Result=Component To Local(ComponentPose=@n2)) {layerGroup=None}
  ]}
 EventGraph {nodes [
   Event BlueprintUpdateAnimation()
   > n1 = Sequence()
   n2 = Get bIsPlayingMeleeAnim()
   > Set Is Melee Anim Float(Is Melee Anim Float:float="To Float (Boolean)"(InBool=NOT Boolean(A=@n2))) -> (Output_Get:float)
   n3 = float * float(A=GetGameTimeInSeconds(), B=1.2) {is=title K2Node_PromotableOperator}
   n4 = "Sin (Radians)"(A=float * float(A=1.25, B=@n3) {is=title K2Node_CommutativeAssociativeBinaryOperator})
   n5 = "Sin (Radians)"(A=float * float(A=2, B=@n3))
   Set Blade Floating(execute=@n1.then_1, Blade Floating:Vector=MakeVector(Y=float * float(A=@n5, B=0.5), Z=float * float(A=@n4, B=0.8))) -> (Output_Get:Vector)
   > n6 = Set pawnVelocitySpeed(pawnVelocitySpeed:double=Vector Length(A=Get PawnVelocity())) -> (Output_Get:double)
   > n7 = Set pawnVelocitySpeedCutoff(pawnVelocitySpeedCutoff:double=SelectFloat(A=0, B=1, bPickA="float > float"(A=@n6, B=1200) {is=title K2Node_CallFunction})) -> (Output_Get:double)
   > Set SpringAlpha(SpringAlpha:double=float * float!(A=0.01, B=@n7, C:float=Get Is Melee Anim Float()) -> (ReturnValue) {is=title K2Node_PromotableOperator}) -> (Output_Get:double)
   TryGetPawnOwner()
   float * float(A="Sin (Radians)"(A=float * float(A=0.6, B=@n3)), B=0.4) {is=title K2Node_CommutativeAssociativeBinaryOperator}
  ]}
]
variables [
 Blade Floating {type=Vector, default=<0 -0.1698 -0.52>}
 Is Melee Anim Float {type=number}
 pawnVelocitySpeed {type=number}
 pawnVelocitySpeedCutoff {type=number}
 SpringAlpha {type=number}
]
